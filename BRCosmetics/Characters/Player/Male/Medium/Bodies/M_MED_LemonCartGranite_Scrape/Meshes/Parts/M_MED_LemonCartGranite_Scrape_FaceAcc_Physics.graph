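# Physics asset for the LemonCartGranite Scrape male face accessory mesh, simulating two dynamic hair strands, left and right.
# Each side is a seven-capsule chain whose root constrains freely to the head, with swing and twist limits widening toward the tip.
# Capsules on head, neck and clavicles supply body shapes around the hair.
# Two constraints reference left and right MSTR hair bones that have no bodies here.
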PhysicsAsset /BRCosmetics/Characters/Player/Male/Medium/Bodies/M_MED_LemonCartGranite_Scrape/Meshes/Parts/M_MED_LemonCartGranite_Scrape_FaceAcc_Physics
bodies {
 clavicle_l [{shapeType=Capsule, length=16.125, radius=8.63, shapeName=clavicle_l_capsule, center=<7.3779 0 -1.1928>, rotation=<rot -0.0002 -90 102.6551>}]
 clavicle_r [{shapeType=Capsule, length=16.125, radius=8.63, shapeName=clavicle_l_capsule, center=<-7.3779 0 1.1928>, rotation=<rot -0.0002 -90 -77.3449>}]
 dyn_hair_scrape_l_01 [{shapeType=Capsule, length=1.7509, radius=1.5325, shapeName=dyn_hair_scrape_l_01_capsule, center=<1.5 0 0>, rotation=<rot 0 90 90>}]
 dyn_hair_scrape_l_02 [{shapeType=Capsule, length=1.9427, radius=1.55, shapeName=dyn_hair_scrape_l_02_capsule, center=<1.5 0 0>, rotation=<rot 0 90 90>}]
 dyn_hair_scrape_l_03 [{shapeType=Capsule, length=2.567, radius=2.0893, shapeName=dyn_hair_scrape_l_03_capsule, center=<1.5 0 0>, rotation=<rot 0 90 90>}]
 dyn_hair_scrape_l_04 [{shapeType=Capsule, length=1.9726, radius=1.9618, shapeName=dyn_hair_scrape_l_04_capsule, center=<1.5 0 0>, rotation=<rot 0 90 90>}]
 dyn_hair_scrape_l_05 [{shapeType=Capsule, length=2.0323, radius=1.6456, shapeName=dyn_hair_scrape_l_05_capsule, center=<1.5 0 0>, rotation=<rot 0 90 90>}]
 dyn_hair_scrape_l_06 [{shapeType=Capsule, length=1.7008, radius=1.0992, shapeName=dyn_hair_scrape_l_06_capsule, center=<1.5 0 0>, rotation=<rot 0 90 90>}]
 dyn_hair_scrape_l_07 [{shapeType=Capsule, length=1.2588, radius=0.6531, shapeName=dyn_hair_scrape_l_07_capsule, center=<1.5 0 0>, rotation=<rot 0 90 90>}]
 dyn_hair_scrape_r_01 [{shapeType=Capsule, length=1.749, radius=1.521, shapeName=dyn_hair_scrape_r_01_capsule, center=<-1.5 0 0>, rotation=<rot 0 90 90>}]
 dyn_hair_scrape_r_02 [{shapeType=Capsule, length=1.9429, radius=1.5502, shapeName=dyn_hair_scrape_r_02_capsule, center=<-1.5 0 0>, rotation=<rot 0 90 90>}]
 dyn_hair_scrape_r_03 [{shapeType=Capsule, length=2.567, radius=2.0893, shapeName=dyn_hair_scrape_r_03_capsule, center=<-1.5 0 0>, rotation=<rot 0 90 90>}]
 dyn_hair_scrape_r_04 [{shapeType=Capsule, length=1.9726, radius=1.9618, shapeName=dyn_hair_scrape_r_04_capsule, center=<-1.5 0 0>, rotation=<rot 0 90 90>}]
 dyn_hair_scrape_r_05 [{shapeType=Capsule, length=2.0323, radius=1.6456, shapeName=dyn_hair_scrape_r_05_capsule, center=<-1.5 0 0>, rotation=<rot 0 90 90>}]
 dyn_hair_scrape_r_06 [{shapeType=Capsule, length=1.7008, radius=1.0992, shapeName=dyn_hair_scrape_r_06_capsule, center=<-1.5 0 0>, rotation=<rot 0 90 90>}]
 dyn_hair_scrape_r_07 [{shapeType=Capsule, length=1.2588, radius=0.6531, shapeName=dyn_hair_scrape_r_07_capsule, center=<-1.5 0 0>, rotation=<rot 0 90 90>}]
 head [
  {
   shapeType=Capsule
   length=2.6404
   radius=7.6437
   shapeName=head_capsule
   center=<6.2728 2.1828 0.0092>
   rotation=<rot -0.2469 -87.1356 -89.9541>
  }
  {shapeType=Capsule, length=14.75, radius=7.25, shapeName=head_capsule, center=<0 12.5516 0.2253>, rotation=<rot 0.2 11.2088 88.9911>}
 ]
 neck_01 [{
   shapeType=Capsule
   length=7.9375
   radius=5.8175
   shapeName=neck_01_capsule
   center=<4.0339 -0.0001 0.0043>
   rotation=<rot 0 -90.0019 89.9387>
  }]
}
constraints [
 {
  bone1Name=dyn_hair_scrape_l_01
  bone2Name=head
  swing1LimitDegrees=45
  swing1Motion=Free
  swing2LimitDegrees=45
  swing2Motion=Free
  twistLimitDegrees=45
  twistMotion=Free
 }
 {bone1Name=dyn_hair_scrape_l_02, bone2Name=dyn_hair_scrape_l_01, swing1LimitDegrees=8, swing2LimitDegrees=8, twistLimitDegrees=4}
 {bone1Name=dyn_hair_scrape_l_03, bone2Name=dyn_hair_scrape_l_02, swing1LimitDegrees=8, swing2LimitDegrees=8, twistLimitDegrees=4}
 {bone1Name=dyn_hair_scrape_l_04, bone2Name=dyn_hair_scrape_l_03, swing1LimitDegrees=24, swing2LimitDegrees=24, twistLimitDegrees=8}
 {bone1Name=dyn_hair_scrape_l_05, bone2Name=dyn_hair_scrape_l_04, swing1LimitDegrees=32, swing2LimitDegrees=32, twistLimitDegrees=12}
 {bone1Name=dyn_hair_scrape_l_06, bone2Name=dyn_hair_scrape_l_05, swing1LimitDegrees=32, swing2LimitDegrees=32, twistLimitDegrees=12}
 {bone1Name=dyn_hair_scrape_l_07, bone2Name=dyn_hair_scrape_l_06, swing1LimitDegrees=32, swing2LimitDegrees=32, twistLimitDegrees=12}
 {
  bone1Name=dyn_hair_scrape_r_01
  bone2Name=head
  swing1LimitDegrees=45
  swing1Motion=Free
  swing2LimitDegrees=45
  swing2Motion=Free
  twistLimitDegrees=45
  twistMotion=Free
 }
 {bone1Name=dyn_hair_scrape_r_02, bone2Name=dyn_hair_scrape_r_01, swing1LimitDegrees=8, swing2LimitDegrees=8, twistLimitDegrees=4}
 {bone1Name=dyn_hair_scrape_r_03, bone2Name=dyn_hair_scrape_r_02, swing1LimitDegrees=8, swing2LimitDegrees=8, twistLimitDegrees=4}
 {bone1Name=dyn_hair_scrape_r_04, bone2Name=dyn_hair_scrape_r_03, swing1LimitDegrees=24, swing2LimitDegrees=24, twistLimitDegrees=8}
 {bone1Name=dyn_hair_scrape_r_05, bone2Name=dyn_hair_scrape_r_04, swing1LimitDegrees=32, swing2LimitDegrees=32, twistLimitDegrees=12}
 {bone1Name=dyn_hair_scrape_r_06, bone2Name=dyn_hair_scrape_r_05, swing1LimitDegrees=32, swing2LimitDegrees=32, twistLimitDegrees=12}
 {bone1Name=dyn_hair_scrape_r_07, bone2Name=dyn_hair_scrape_r_06, swing1LimitDegrees=32, swing2LimitDegrees=32, twistLimitDegrees=12}
 {
  bone1Name=head
  bone2Name=neck_01
  swing1LimitDegrees=45
  swing1Motion=Free
  swing2LimitDegrees=45
  swing2Motion=Free
  twistLimitDegrees=45
  twistMotion=Free
 }
 {
  bone1Name=dyn_hair_l_MSTR
  bone2Name=head
  swing1LimitDegrees=45
  swing1Motion=Free
  swing2LimitDegrees=45
  swing2Motion=Free
  twistLimitDegrees=45
  twistMotion=Free
 }
 {
  bone1Name=dyn_hair_r_MSTR
  bone2Name=head
  swing1LimitDegrees=45
  swing1Motion=Free
  swing2LimitDegrees=45
  swing2Motion=Free
  twistLimitDegrees=45
  twistMotion=Free
 }
]
